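# MF_ReliableElk is a material function for the ReliableElk skin of the M_MED_Lettuce player body that optionally adds a sequin effect.
# It passes incoming material attributes and the skin's body FX mask texture into the shared weapon-skin function MF_Sequins.
# The EnableSequin static switch outputs either that result or the unmodified input attributes.
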MaterialFunction /BRCosmetics/Characters/Player/Male/Medium/Bodies/M_MED_Lettuce/Skins/ReliableElk/Materials/MF_ReliableElk
graph {nodes [
  n1 = TextureObject() {
   samplerType=SAMPLERTYPE_Masks
   texture=/BRCosmetics/Characters/Player/Male/Medium/Bodies/M_MED_Lettuce/Skins/ReliableElk/Textures/T_ReliableElk_Body_FX
  }
  n2 = FunctionInput() {inputType=FunctionInput_MaterialAttributes}
  FunctionOutput(None=StaticSwitchParameter(True=MaterialFunctionCall!(Mask=@n1, inMat=@n2) {materialFunction=/Game/Weapons/WeaponSkins/MaterialFunction/MF_Sequins}.Result, False=@n2) {parameterName=EnableSequin})
 ]}
